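# Work-in-progress test Animation Blueprint for the PistolCanary tentacle boss.
# A state machine holds idle, attack, turn, damage and death clips, but only Idle to GroundSlam, Idle to Energy Whip and Death to Idle are linked, each by a bool.
# Its pose feeds an impact-deform Control Rig driven by hit location, trace-start and bone arrays.
# Update clears those arrays when bIsHit is false; the boss cast is unwired.
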AnimBlueprint /PistolCanary/Anim/Test/ABP_PistolCanary_TentacleBoss_WIP
target_skeleton=/PistolCanary/Assets/Meshes/Tentacle_PistolCanary_Creature_Skeleton
graphs [
 AnimGraph {nodes [
   AnimGraphNode_StateMachine_0 = Tentacle Boss SM() {is=title AnimGraphNode_StateMachine}
   n1 = Control Rig!(
    Source=@AnimGraphNode_StateMachine_0
    Alpha
    HitLocationsWS:Array of Vectors=Get HitLocationsWS() -> (HitLocationsWS:Array of Vectors)
    HitTraceStartsWS:Array of Vectors=Get HitTraceStartsWS() -> (HitTraceStartsWS:Array of Vectors)
    HitBones:Array of Names=Get Hit Bones() -> (Hit Bones:Array of Names)
    ResponseSpeed_Normalized:double=0
    Radius_Normalized:double=0
    Stiffness_Normalized:double=0
    Force_Normalized:double=0
    MaxDuration:double=1
   ) -> (Pose) {
    alphaBoolBlend {blendOption=Linear}
    controlRigAssetReference {blueprintRigClass=/PistolCanary/Anim/Test/PistolCanary_Tentacle_ImpactDeformer/CR_PistolCanary_TentacleBoss_ImpactDeform.CR_PistolCanary_TentacleBoss_ImpactDeform_C}
   }
   Output Pose(Result=@n1)
  ]}
 Tentacle Boss SM {
  entry=Idle
  path=AnimGraph.AnimGraphNode_StateMachine_0.Tentacle Boss SM
  ids=id
  nodes [
   AnimStateNode_0 = Idle()
   AnimStateNode_1 = CircularBeam() {is=title AnimStateNode}
   AnimStateNode_10 = DamageReactA() {is=title AnimStateNode}
   AnimStateNode_11 = DamageReactB() {is=title AnimStateNode}
   AnimStateNode_12 = TurnL Idle() {is=title AnimStateNode}
   AnimStateNode_13 = Homing Burst() {is=title AnimStateNode}
   AnimStateNode_14 = Energy Whip() {is=title AnimStateNode}
   AnimStateNode_15 = Death() {is=title AnimStateNode}
   AnimStateNode_2 = EnergyBlast() {is=title AnimStateNode}
   AnimStateNode_3 = Turn Left() {is=title AnimStateNode}
   AnimStateNode_4 = Turn Right() {is=title AnimStateNode}
   AnimStateNode_5 = TurnR Return() {is=title AnimStateNode}
   AnimStateNode_6 = TurnL Return() {is=title AnimStateNode}
   AnimStateNode_7 = GroundSlam() {is=title AnimStateNode}
   AnimStateNode_8 = GroundSlamJab() {is=title AnimStateNode}
   AnimStateNode_9 = TurnR Idle() {is=title AnimStateNode}
   AnimStateTransitionNode_15 = AnimStateTransitionNode()
   AnimStateTransitionNode_17 = AnimStateTransitionNode()
   AnimStateTransitionNode_18 = AnimStateTransitionNode()
  ]
  links [
   {
    dst=Idle
    rule=AnimGraph.AnimGraphNode_StateMachine_0.Tentacle Boss SM.AnimStateTransitionNode_15.Transition
    src=Death
    via=AnimStateTransitionNode_15
   }
   {
    dst=Energy Whip
    rule=AnimGraph.AnimGraphNode_StateMachine_0.Tentacle Boss SM.AnimStateTransitionNode_18.Transition
    src=Idle
    via=AnimStateTransitionNode_18
   }
   {
    dst=GroundSlam
    rule=AnimGraph.AnimGraphNode_StateMachine_0.Tentacle Boss SM.AnimStateTransitionNode_17.Transition
    src=Idle
    via=AnimStateTransitionNode_17
   }
  ]
 }
 Idle {
  path=AnimGraph.AnimGraphNode_StateMachine_0.Tentacle Boss SM.AnimStateNode_0.Idle
  nodes [Output Animation Pose(Result=Sequence Player 'PistolCanary_Tentacle_Idle_01'() {sequence=/PistolCanary/Anim/SingleTentacle/Section02/PistolCanary_Tentacle_Idles/PistolCanary_Tentacle_Idle_01})]
 }
 CircularBeam {
  path=AnimGraph.AnimGraphNode_StateMachine_0.Tentacle Boss SM.AnimStateNode_1.CircularBeam
  nodes [
   n1 = Sequence Player 'PistolCanary_Tentacle_Attack_CircularBeam'() {
    bLoopAnimation=false
    sequence=/PistolCanary/Anim/SingleTentacle/Section02/PistolCanary_Attack_CircularBeam/PistolCanary_Tentacle_Attack_CircularBeam
   }
   Output Animation Pose(Result=@n1)
  ]
 }
 DamageReactA {
  path=AnimGraph.AnimGraphNode_StateMachine_0.Tentacle Boss SM.AnimStateNode_10.DamageReactA
  nodes [
   n1 = Sequence Player 'PistolCanary_Tentacle_Damage_Thrashing'() {
    bLoopAnimation=false
    sequence=/PistolCanary/Anim/SingleTentacle/Section02/PistolCanary_Tentacle_HitReacts/PistolCanary_Tentacle_Damage_Thrashing
   }
   Output Animation Pose(Result=@n1)
  ]
 }
 DamageReactB {
  path=AnimGraph.AnimGraphNode_StateMachine_0.Tentacle Boss SM.AnimStateNode_11.DamageReactB
  nodes [
   n1 = Sequence Player 'PistolCanary_Tentacle_Damage_Thrashing_02'() {
    bLoopAnimation=false
    sequence=/PistolCanary/Anim/SingleTentacle/Section02/PistolCanary_Tentacle_HitReacts/PistolCanary_Tentacle_Damage_Thrashing_02
   }
   Output Animation Pose(Result=@n1)
  ]
 }
 TurnL Idle {
  path=AnimGraph.AnimGraphNode_StateMachine_0.Tentacle Boss SM.AnimStateNode_12.TurnL Idle
  nodes [
   n1 = Sequence Player 'PistolCanary_Tentacle_Turn_180L_Idle'() {bLoopAnimation=false, sequence=/PistolCanary/Anim/SingleTentacle/Unused/PistolCanary_Tentacle_Turn_180L/PistolCanary_Tentacle_Turn_180L_Idle}
   Output Animation Pose(Result=@n1)
  ]
 }
 Homing Burst {
  path=AnimGraph.AnimGraphNode_StateMachine_0.Tentacle Boss SM.AnimStateNode_13.Homing Burst
  nodes [
   n1 = Sequence Player 'PistolCanary_Tentacle_Attack_HomingBurst'() {
    bLoopAnimation=false
    sequence=/PistolCanary/Anim/SingleTentacle/Section02/PistolCanary_Attack_HomingBurst/PistolCanary_Tentacle_Attack_HomingBurst
   }
   Output Animation Pose(Result=@n1)
  ]
 }
 Energy Whip {
  path=AnimGraph.AnimGraphNode_StateMachine_0.Tentacle Boss SM.AnimStateNode_14.Energy Whip
  nodes [
   n1 = Sequence Player 'PistolCanary_Tentacle_Attack_EnergyWhip'() {
    bLoopAnimation=false
    sequence=/PistolCanary/Anim/SingleTentacle/Section02/PistolCanary_Attack_EnergyWhip/PistolCanary_Tentacle_Attack_EnergyWhip
   }
   Output Animation Pose(Result=@n1)
  ]
 }
 Death {
  path=AnimGraph.AnimGraphNode_StateMachine_0.Tentacle Boss SM.AnimStateNode_15.Death
  nodes [Output Animation Pose(Result=Sequence Player 'PistolCanary_Tentacle_Death_01'() {bLoopAnimation=false, sequence=/PistolCanary/Anim/SingleTentacle/Section03/PistolCanary_Tentacle_Death_01})]
 }
 EnergyBlast {
  path=AnimGraph.AnimGraphNode_StateMachine_0.Tentacle Boss SM.AnimStateNode_2.EnergyBlast
  nodes [
   n1 = Sequence Player 'PistolCanary_Tentacle_Attack_EnergyBlast'() {
    bLoopAnimation=false
    sequence=/PistolCanary/Anim/SingleTentacle/Section02/PistolCanary_Attack_EnergyBlast/PistolCanary_Tentacle_Attack_EnergyBlast
   }
   Output Animation Pose(Result=@n1)
  ]
 }
 Turn Left {
  path=AnimGraph.AnimGraphNode_StateMachine_0.Tentacle Boss SM.AnimStateNode_3.Turn Left
  nodes [
   n1 = Sequence Player 'PistolCanary_Tentacle_Turn_180L'() {bLoopAnimation=false, sequence=/PistolCanary/Anim/SingleTentacle/Unused/PistolCanary_Tentacle_Turn_180L/PistolCanary_Tentacle_Turn_180L}
   Output Animation Pose(Result=@n1)
  ]
 }
 Turn Right {
  path=AnimGraph.AnimGraphNode_StateMachine_0.Tentacle Boss SM.AnimStateNode_4.Turn Right
  nodes [
   n1 = Sequence Player 'PistolCanary_Tentacle_Turn_180R'() {bLoopAnimation=false, sequence=/PistolCanary/Anim/SingleTentacle/Unused/PistolCanary_Tentacle_Turn_180R/PistolCanary_Tentacle_Turn_180R}
   Output Animation Pose(Result=@n1)
  ]
 }
 TurnR Return {
  path=AnimGraph.AnimGraphNode_StateMachine_0.Tentacle Boss SM.AnimStateNode_5.TurnR Return
  nodes [
   n1 = Sequence Player 'PistolCanary_Tentacle_Turn_180R_Return'() {
    bLoopAnimation=false
    sequence=/PistolCanary/Anim/SingleTentacle/Unused/PistolCanary_Tentacle_Turn_180R/PistolCanary_Tentacle_Turn_180R_Return
   }
   Output Animation Pose(Result=@n1)
  ]
 }
 TurnL Return {
  path=AnimGraph.AnimGraphNode_StateMachine_0.Tentacle Boss SM.AnimStateNode_6.TurnL Return
  nodes [
   n1 = Sequence Player 'PistolCanary_Tentacle_Turn_180L_Return'() {
    bLoopAnimation=false
    sequence=/PistolCanary/Anim/SingleTentacle/Unused/PistolCanary_Tentacle_Turn_180L/PistolCanary_Tentacle_Turn_180L_Return
   }
   Output Animation Pose(Result=@n1)
  ]
 }
 GroundSlam {
  path=AnimGraph.AnimGraphNode_StateMachine_0.Tentacle Boss SM.AnimStateNode_7.GroundSlam
  nodes [
   n1 = Sequence Player 'PistolCanary_Tentacle_Attack_GroundSlam'() {
    bLoopAnimation=false
    sequence=/PistolCanary/Anim/SingleTentacle/Section02/PistolCanary_Attack_GroundSlam/PistolCanary_Tentacle_Attack_GroundSlam
   }
   Output Animation Pose(Result=@n1)
  ]
 }
 GroundSlamJab {
  path=AnimGraph.AnimGraphNode_StateMachine_0.Tentacle Boss SM.AnimStateNode_8.GroundSlamJab
  nodes [
   n1 = Sequence Player 'PistolCanary_Tentacle_Attack_GroundSlamJab'() {
    bLoopAnimation=false
    sequence=/PistolCanary/Anim/SingleTentacle/Section02/PistolCanary_Attack_GroundSlamJab/PistolCanary_Tentacle_Attack_GroundSlamJab
   }
   Output Animation Pose(Result=@n1)
  ]
 }
 TurnR Idle {
  path=AnimGraph.AnimGraphNode_StateMachine_0.Tentacle Boss SM.AnimStateNode_9.TurnR Idle
  nodes [
   n1 = Sequence Player 'PistolCanary_Tentacle_Turn_180R_Idle'() {bLoopAnimation=false, sequence=/PistolCanary/Anim/SingleTentacle/Unused/PistolCanary_Tentacle_Turn_180R/PistolCanary_Tentacle_Turn_180R_Idle}
   Output Animation Pose(Result=@n1)
  ]
 }
 Transition {
  path=AnimGraph.AnimGraphNode_StateMachine_0.Tentacle Boss SM.AnimStateTransitionNode_15.Transition
  nodes [Result(bCanEnterTransition=Get bIsDead() -> (bIsDead:bool))]
 }
 Transition {
  path=AnimGraph.AnimGraphNode_StateMachine_0.Tentacle Boss SM.AnimStateTransitionNode_17.Transition
  nodes [Result(bCanEnterTransition=Get bCanGroundSlam() -> (bCanGroundSlam:bool))]
 }
 Transition {
  path=AnimGraph.AnimGraphNode_StateMachine_0.Tentacle Boss SM.AnimStateTransitionNode_18.Transition
  nodes [Result(bCanEnterTransition=Get bCanEnergyWhip() -> (bCanEnergyWhip:bool))]
 }
 EventGraph {nodes [
   Cast To PistolCanary_TentacleBoss!(execute, Object:Wildcard) -> (then, CastFailed, AsPistol Canary Tentacle Boss:Pistol Canary Tentacle Boss Object Reference) {is=title K2Node_DynamicCast}
   > Set bCanGroundSlam(bCanGroundSlam:bool=false) -> (Output_Get:bool)
   Event BlueprintUpdateAnimation()
   > n1 = Branch(Condition=Get bIsHit() -> (bIsHit:bool))
   Clear(execute=@n1.else, TargetArray:Array of Vectors=Get HitLocationsWS())
   > Clear(TargetArray=Get HitTraceStartsWS())
   > Clear(TargetArray:Array of Names=Get Hit Bones())
   TryGetPawnOwner()
  ]}
]
variables [
 bCanEnergyWhip {type=boolean}
 bCanGroundSlam {type=boolean}
 bIsDead {type=boolean}
 bIsHit {type=boolean}
 Hit Bones {type="array<string>", default []}
 HitLocationsWS {type="array<Vector>"}
 Hits {type="array<Transform>", default []}
 HitTraceStartsWS {type="array<Vector>"}
]
