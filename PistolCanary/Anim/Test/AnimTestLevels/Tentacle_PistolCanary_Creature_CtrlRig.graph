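# Level Sequence in an animation test folder that sets up the Tentacle Pistol Canary Creature skeletal mesh actor with its Control Rig, at 30 fps.
# The actor's binding has spawn and transform tracks, an empty skeletal animation track, and a Control Rig parameter track for Tentacle_PistolCanary_Creature_CtrlRig.
# The rig section exposes only a "control rig options" transform, and all channels show just default values, with no keyed animation.
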LevelSequence /PistolCanary/Anim/Test/AnimTestLevels/Tentacle_PistolCanary_Creature_CtrlRig
bindings [Tentacle Pistol Canary Creature {
  id=#1
  possessed_class=SkeletalMeshActor
  tracks [
   {type=MovieSceneSpawnTrack, sections [{type=MovieSceneSpawnSection, channels [None {type=MovieSceneScriptingBoolChannel, default=true}]}]}
   {
    type=MovieScene3DTransformTrack
    sections [{
      type=MovieScene3DTransformSection
      channels [
       Location.X {type=MovieSceneScriptingDoubleChannel, default=0}
       Location.Y {type=MovieSceneScriptingDoubleChannel, default=0}
       Location.Z {type=MovieSceneScriptingDoubleChannel, default=0}
       Rotation.X {type=MovieSceneScriptingDoubleChannel, default=0}
       Rotation.Y {type=MovieSceneScriptingDoubleChannel, default=0}
       Rotation.Z {type=MovieSceneScriptingDoubleChannel, default=0}
       Scale.X {type=MovieSceneScriptingDoubleChannel, default=1}
       Scale.Y {type=MovieSceneScriptingDoubleChannel, default=1}
       Scale.Z {type=MovieSceneScriptingDoubleChannel, default=1}
      ]
     }]
   }
   {type=MovieSceneSkeletalAnimationTrack, sections []}
   {
    type=MovieSceneControlRigParameterTrack
    display_name=Tentacle_PistolCanary_Creature_CtrlRig
    sections [{
      type=MovieSceneControlRigParameterSection
      channels [
       control rig options.Location.X {type=MovieSceneScriptingFloatChannel, default=0}
       control rig options.Location.Y {type=MovieSceneScriptingFloatChannel, default=0}
       control rig options.Location.Z {type=MovieSceneScriptingFloatChannel, default=0}
       control rig options.Rotation.X {type=MovieSceneScriptingFloatChannel, default=0}
       control rig options.Rotation.Y {type=MovieSceneScriptingFloatChannel, default=0}
       control rig options.Rotation.Z {type=MovieSceneScriptingFloatChannel, default=0}
       control rig options.Scale.X {type=MovieSceneScriptingFloatChannel, default=1}
       control rig options.Scale.Y {type=MovieSceneScriptingFloatChannel, default=1}
       control rig options.Scale.Z {type=MovieSceneScriptingFloatChannel, default=1}
      ]
     }]
   }
  ]
 }]
display_rate {denominator=1, numerator=30}
tick_resolution {denominator=1, numerator=24000}
tracks []
